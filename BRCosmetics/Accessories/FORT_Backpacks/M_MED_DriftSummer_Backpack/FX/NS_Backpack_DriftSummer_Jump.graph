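# Niagara jump effect for the DriftSummer backpack cosmetic, emitting short-lived falling drip sprites from sockets.
# A single GPU emitter spawns translucent sprites continuously, placing them through a system-level Socket Reader.
# An initial forward and downward velocity plus strong gravity pulls the drips down while alpha fades.
# Size curves reshape each sprite just before death, and a time-of-day scalar dims color at night and evening.
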NiagaraSystem /BRCosmetics/Accessories/FORT_Backpacks/M_MED_DriftSummer_Backpack/FX/NS_Backpack_DriftSummer_Jump
dependencies {
 usedDataInterfaces [
  &NiagaraDataInterfaceCurve
  &NiagaraDataInterfaceVector2DCurve
  &NiagaraDataInterfaceSocketReader
 ]
 usedDynamicInputs [&RandomRangeFloat]
 usedModules [
  &self:SystemSpawnScript.SystemScriptSource.SystemScriptGraph.NiagaraNodeAssignment_0.SetVariables_#1
  &SystemState
 ]
}
emitters {Sprite_Drips {
  simTarget=GPUComputeSim
  emitterUpdateScript [
   &EmitterState()
   &SpawnRate(SpawnRate=240, Use Spawn Probability=false)
  ]
  particleSpawnScript [
   &InitializeParticle(
    Color=<rgba 1 1 1 0.35>
    Lifetime Max=0.3
    Lifetime Min=0.1
    Lifetime Mode=::NewEnumerator1
    Sprite Size Max=<2 15>
    Sprite Size Min=<0.5 5>
    Sprite Size Mode=::NewEnumerator2
    UsePositionOffset=false
    Write Material Random=true
   )
   &AddVelocity(Velocity=<120 0 -150>)
   &SocketLocation(SamplePosOffset=false, Socket Reader=$System.Socket Reader, Socket Sampling Mode=::NewEnumerator3, VelocitySpeedLimit=false)
  ]
  particleUpdateScript [
   &ParticleState()
   &ScaleSpriteSize(
    Non-Uniform Curve Sprite Scale=DataInterface() {propertyValues {
      XCurve {keys [<at 0.9 1.2 RCIM_Cubic>, {arriveTangent=-12, interpMode=RCIM_Cubic, leaveTangent=-12, time=0.91, value=0.79}, <at 1 0 RCIM_Cubic>]}
      YCurve {keys [
        {arriveTangent=-1.3596, interpMode=RCIM_Cubic, time=0.9, value=0.59}
        {arriveTangent=-1.2058, interpMode=RCIM_Cubic, leaveTangent=-1.2058, tangentMode=RCTM_User, time=0.91, value=2.28}
        <at 1 0 RCIM_Cubic>
       ]}
     }}
    Scale Sprite Size Mode=::NewEnumerator3
   )
   &ScaleColor(
    Scale Alpha=&FloatFromCurve(FloatCurve=DataInterface() {propertyValues {Curve {keys [<at -0.0046 1.0113 RCIM_Cubic>,<at 1.0011 0.002 RCIM_Cubic>]}}})
    Scale RGB=&VectorFromFloat(Value=&FNTimeOfDayScalar(Night Scale=0.75, Morning Scale=1, Evening Scale=0.8))
    ScaleA=true
    ScaleRGB=true
   )
   &GravityForce(Gravity=<0 0 -2000>)
   &SolveForcesAndVelocity(Clamp Velocity=false, Limit Acceleration=false)
  ]
  rendererClasses [&NiagaraSpriteRendererProperties]
  renderers [{rendererClass=&NiagaraSpriteRendererProperties}]
 }}
summary {
 systemName=NS_Backpack_DriftSummer_Jump
 emitters [{emitterName=Sprite_Drips, simTarget=GPUComputeSim, rendererClasses [&NiagaraSpriteRendererProperties]}]
 userVariables []
}
