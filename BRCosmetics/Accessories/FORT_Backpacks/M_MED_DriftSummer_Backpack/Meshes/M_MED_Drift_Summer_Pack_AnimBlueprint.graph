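# Animation Blueprint for the Drift Summer backpack cosmetic, adding procedural tilt and secondary motion.
# The event graph maps spine_05-to-spine_01 yaw delta on the owner mesh to an additive root roll.
# From ref pose, AnimDynamics springs dyn_trigger linearly and dyn_water_01 angularly, then Bone Driven Controllers turn dyn_water_01's rotation into translation on four other water bones, faking slosh.
# The dyn_loader AnimDynamics node is unconnected.
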AnimBlueprint /BRCosmetics/Accessories/FORT_Backpacks/M_MED_DriftSummer_Backpack/Meshes/M_MED_Drift_Summer_Pack_AnimBlueprint
target_skeleton=&/M_MED_Drift_Summer_Pack_Skeleton
graphs [
 AnimGraph {nodes [
   Anim Dynamics - Bone: dyn_loader(Alpha=1, ExternalForce="0,0,0") {
    bGravityOverrideInSimSpace=true
    bLinearSpring=true
    bOverrideLinearDamping=true
    bUseGravityOverride=true
    linearDampingOverride=0.95
    linearSpringConstant=150
    alphaBoolBlend {blendOption=Linear}
    boundBone {boneName=dyn_loader}
    componentAppliedLinearAccClamp=<1000 1000 1000>
    componentLinearAccScale=<0.8 0.8 1>
    componentLinearVelScale=<0.1 0.1 0.1>
    gravityOverride {z=-980}
    physicsBodyDefinitions [{
      sphereCollisionRadius=0
      boundBone {boneName=dyn_loader}
      boxExtents=<10 5 5>
      constraintSetup {angularLimitsMax=<0 0 0>, angularLimitsMin=<0 0 0>, angularTarget=<0 0 0>, linearAxesMax=<0.5 0 0>, linearAxesMin=<-2.5 0 0>}
      localJointOffset=<0 0 0>
     }]
   }
   n1 = "Transform (Modify) Bone - Bone: root"(ComponentPose=Local To Component(LocalPose=Local Space Ref Pose()), Translation="0,0,0", Rotation=Get backpack_rotate()) {rotationMode=BMM_Additive, rotationSpace=BCS_BoneSpace, alphaBoolBlend {blendOption=Linear}, boneToModify {boneName=root}}
   n2 = Anim Dynamics - Bone: dyn_trigger(ComponentPose=@n1, Alpha=1, ExternalForce="0,0,0") {
    bGravityOverrideInSimSpace=true
    bLinearSpring=true
    bUseGravityOverride=true
    linearSpringConstant=100
    alphaBoolBlend {blendOption=Linear}
    boundBone {boneName=dyn_trigger}
    componentAppliedLinearAccClamp=<1000 1000 1000>
    componentLinearAccScale=<0.35 0.35 0.35>
    gravityOverride {z=-980}
    physicsBodyDefinitions [{
      sphereCollisionRadius=0
      boundBone {boneName=dyn_trigger}
      boxExtents=<6 3 3>
      constraintSetup {angularLimitsMax=<0 0 0>, angularLimitsMin=<0 0 0>, angularTarget=<0 0 0>, linearAxesMax=<0.3 0 0>, linearAxesMin=<-0.3 0 0>}
      localJointOffset=<0 0 0>
     }]
   }
   n3 = Anim Dynamics - Bone: dyn_water_01(ComponentPose=@n2, Alpha=1, ExternalForce="0,0,0") {
    angularSpringConstant=500
    bAngularSpring=true
    bGravityOverrideInSimSpace=true
    bUseGravityOverride=true
    alphaBoolBlend {blendOption=Linear}
    boundBone {boneName=dyn_water_01}
    componentAppliedLinearAccClamp=<3000 3000 3000>
    componentLinearAccScale=<0.4 0.4 0.4>
    componentLinearVelScale=<0.2 0.2 0.2>
    gravityOverride {z=-980}
    physicsBodyDefinitions [{
      sphereCollisionRadius=0
      boundBone {boneName=dyn_water_01}
      boxExtents=<10 10 10>
      constraintSetup {angularLimitsMax=<10 45 45>, angularLimitsMin=<-10 -45 -45>, angularTarget=<0 0 0>}
      localJointOffset=<2 0 0>
     }]
   }
   n4 = "dyn_water_02.translateX = dyn_water_01.rotateY * -0.12 - Bone Driven Controller"!(ComponentPose:Component Space Pose Link Structure=@n3, Alpha:float=1) -> (Pose:Component Space Pose Link Structure) {
    is=title AnimGraphNode_BoneDrivenController
    bAffectTargetTranslationX=true
    modificationMode=AddToRefPose
    multiplier=-0.12
    sourceComponent=RotationY
    alphaBoolBlend {blendOption=Linear}
    sourceBone {boneName=dyn_water_01}
    targetBone {boneName=dyn_water_02}
   }
   n5 = "dyn_water_04.translateX = dyn_water_01.rotateY * 0.08 - Bone Driven Controller"!(ComponentPose:Component Space Pose Link Structure=@n4, Alpha:float=1) -> (Pose:Component Space Pose Link Structure) {
    is=title AnimGraphNode_BoneDrivenController
    bAffectTargetTranslationX=true
    modificationMode=AddToRefPose
    multiplier=0.08
    sourceComponent=RotationY
    alphaBoolBlend {blendOption=Linear}
    sourceBone {boneName=dyn_water_01}
    targetBone {boneName=dyn_water_04}
   }
   n6 = "dyn_water_03.translateX = dyn_water_01.rotateZ * 0.1 - Bone Driven Controller"!(ComponentPose:Component Space Pose Link Structure=@n5, Alpha:float=1) -> (Pose:Component Space Pose Link Structure) {
    is=title AnimGraphNode_BoneDrivenController
    bAffectTargetTranslationX=true
    modificationMode=AddToRefPose
    multiplier=0.1
    sourceComponent=RotationZ
    alphaBoolBlend {blendOption=Linear}
    sourceBone {boneName=dyn_water_01}
    targetBone {boneName=dyn_water_03}
   }
   n7 = "dyn_water_05.translateX = dyn_water_01.rotateZ * -0.03 - Bone Driven Controller"!(ComponentPose:Component Space Pose Link Structure=@n6, Alpha:float=1) -> (Pose:Component Space Pose Link Structure) {
    is=title AnimGraphNode_BoneDrivenController
    bAffectTargetTranslationX=true
    modificationMode=AddToRefPose
    multiplier=-0.03
    sourceComponent=RotationZ
    alphaBoolBlend {blendOption=Linear}
    sourceBone {boneName=dyn_water_01}
    targetBone {boneName=dyn_water_05}
   }
   Output Pose(Result=Component To Local(ComponentPose=@n7)) {layerGroup=None}
  ]}
 EventGraph {nodes [
   Event BlueprintUpdateAnimation()
   n1 = GetOwnerSkeletalMeshComponent()
   n2 = BreakRotator(InRot=BreakTransform(InTransform=GetDeltaTransformFromRefPose(self=@n1, BoneName=spine_05, BaseName=spine_01)).Rotation)
   > Set backpack_rotate(backpack_rotate=MakeRotator(Roll=MapRangeClamped(Value=@n2.Yaw, InRangeA=0, InRangeB=-30, OutRangeB=20)))
   TryGetPawnOwner()
  ]}
]
variables [backpack_rotate {type=Rotator}]
